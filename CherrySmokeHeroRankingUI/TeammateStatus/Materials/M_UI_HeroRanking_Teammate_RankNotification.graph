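# Translucent UI material that draws an animated orange hexagon-and-circle graphic for the teammate rank notification in Hero Ranking.
# The hexagon is an anti-aliased polygon SDF with perspective-skewed, slightly squashed UVs, alongside a circle from a shared UI function.
# Nested lerps between three orange tones supply color, and the max of two masks supplies opacity.
# Scalar parameters animate vertical translation through a Panner, center-based UV scale, and opacity.
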Material /CherrySmokeHeroRankingUI/TeammateStatus/Materials/M_UI_HeroRanking_Teammate_RankNotification
graph {
 properties {blendMode=BLEND_Translucent, materialDomain=MD_UI, pixelDepthOffsetMode=PDOM_Legacy, usedShadingModels=Default Lit}
 nodes [
  n1 = Add(A=NamedRerouteUsage()) {constB=0.5}
  n2 = NamedRerouteUsage()
  LinearInterpolate(
   A=Constant3Vector() {constant=<rgba 1 0.1146 0.0202 1>}
   B=LinearInterpolate(
    A=Constant3Vector() {constant=<rgba 0.8714 0.0999 0.0176 1>}
    B=Constant3Vector() {constant=<rgba 0.4287 0.0369 0.0052 1>}
    Alpha=Saturate(None=@n1)
   )
   Alpha=OneMinus(None=@n2)
  )
  Multiply(A=Max(A=@n2, B=NamedRerouteUsage()), B=ScalarParameter() {parameterName=Opacity Animation})
  NamedRerouteDeclaration(None=MaterialFunctionCall(Size=Constant() {r=0.1}, GlowMax=Constant() {r=0.9}) {materialFunction=&MF_UI_Circle}.Fill) {name=Circle, nodeColor=<rgba 0 0 0 1>}
  n3 = NamedRerouteUsage()
  n4 = Constant() {r=0.25}
  n5 = Constant() {r=0.1}
  n6 = Add(
   A=Multiply(
    A=MaterialFunctionCall!(UV=@n3, Perspective Rotation=@n4, Perspective Amount=@n5) {materialFunction=/FortUILibrary/MaterialFunctions/Basics/MF_UI_Perspective}.Perspective UVs
    B=Constant2Vector() {g=0.94, r=1}
   )
   B=Constant2Vector() {g=-0.016}
  )
  n7 = Constant() {r=0.5}
  n8 = Constant() {r=6}
  NamedRerouteDeclaration(None=MaterialFunctionCall(
    DistanceField=MaterialFunctionCall!(UVs=@n6, Size=@n7, Sides=@n8, InnerAngle, OutlineThickness, OutlineAlignment, GlowMin, GlowMax) {materialFunction=/FortUILibrary/MaterialFunctions/Shapes/MF_UI_RegularPolygon}.SDF
    Threshold=Constant() {r=0.226}
   ) {materialFunction=&MF_UI_AntiAliasing}.Inner) {name=Hexagon Shape, nodeColor=<rgba 1 0.9609 0 1>}
  n9 = TextureCoordinate()
  n10 = Subtract(A=Multiply(A=ScalarParameter() {defaultValue=0.5, parameterName=Translation Animation}) {constB=2})
  NamedRerouteDeclaration(None=MaterialFunctionCall(UVs=Panner(Coordinate=@n9, Time=@n10) {speedY=0.3}, Texture Scale=ScalarParameter() {defaultValue=1, parameterName=Scale Animation}) {materialFunction=&ScaleUVsByCenter}.UVs) {name=UVs, nodeColor=<rgba 0 0 0 1>}
 ]
}
