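# Translucent UI material for a shine effect on the circular teammate portrait in the hero ranking rank notification.
# ShinePosition offsets the V coordinate of a skewed box band, sized by ShineSize and faded by ShineOpacity.
# A soft circle clips the band, which is max-combined with a circle-minus-circle background glow scaled by BackgroundOpacity.
# The band also blends the Color parameter toward white.
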Material /CherrySmokeHeroRankingUI/TeammateStatus/Materials/M_UI_HeroRanking_Teammate_RankNotification_PortraitShine
graph {
 properties {blendMode=BLEND_Translucent, materialDomain=MD_UI, pixelDepthOffsetMode=PDOM_Legacy, usedShadingModels=Default Lit}
 nodes [
  Constant2Vector() {g=0.08, r=1}
  n1 = TextureCoordinate()
  n2 = Multiply(A=ScalarParameter() {parameterName=ShinePosition}) {constB=2.1}
  n3 = AppendVector(A=ComponentMask(None=@n1) {r=true}, B=Add(A=ComponentMask(None=@n1) {g=true}, B=Subtract(A=@n2) {constB=1.05}) {constB=0})
  n4 = Constant2Vector() {g=-40}
  n5 = Constant() {r=1}
  n6 = ScalarParameter() {defaultValue=0.2, parameterName=ShineSize}
  n7 = Multiply(
   A=MaterialFunctionCall(
    UVs=MaterialFunctionCall!(UV=@n3, Skew Angle=@n4) {materialFunction=/FortUILibrary/MaterialFunctions/Basics/MF_UI_Skew}.Skewed UVs
    Size=AppendVector(A=@n5, B=@n6)
    GlowMax=Constant() {r=0.2}
   ) {materialFunction=&MF_UI_Box}.Fill
   B=ScalarParameter() {parameterName=ShineOpacity}
  )
  LinearInterpolate(A=VectorParameter() {parameterName=Color, defaultValue=<rgba 1 1 1 1>}.RGB, B=Constant3Vector() {constant=<rgba 1 1 1 1>}, Alpha=@n7)
  MaterialFunctionCall() {materialFunction=&DebugTimeSine}
  n8 = MaterialFunctionCall(Size=Constant2Vector() {g=0.85, r=0.85}, GlowMax=Constant() {r=0.1}) {materialFunction=&MF_UI_Circle}
  n9 = MaterialFunctionCall(Size=Constant() {r=0.5}, GlowMax=Constant() {r=1}) {materialFunction=&MF_UI_Circle}
  Max(
   A=Subtract(A=@n7, B=OneMinus(None=@n8.Fill))
   B=Multiply(A=Subtract(A=@n8.Fill, B=@n9.Fill), B=ScalarParameter() {parameterName=BackgroundOpacity})
  )
  Reroute()
 ]
}
