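# Teammate status widget that shows a Hero Ranking rank-increase notification, with a notification icon over a portrait glow.
# On Set Player State, it passes the player state to the CherrySmokeHeroRanking view model as the displayed player.
# Anim_IncreaseRank animates material parameters on both images for the icon's fade and scale pop and the glow's shine sweep.
# The graph never plays that animation, and the widget derives from CherrySmokeHeroRankingPlayerInfoWidget.
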WidgetBlueprint /CherrySmokeHeroRankingUI/TeammateStatus/WBP_CherrySmoke_HeroRanking_TeammateRankNotify
animations [Anim_IncreaseRank {
  binding_count=2
  length_seconds=1.55
  object=&self:Anim_IncreaseRank
  sequence {bindings [
    Image_Notification {
     id=#1
     possessed_class=Image
     tracks [{
       type=MovieSceneWidgetMaterialTrack
       sections [{
         type=MovieSceneParameterSection
         channels [
          Fire Shape Visible {type=MovieSceneScriptingFloatChannel, keys [[20,1],[42,1],[45.038,0]]}
          Opacity Animation {type=MovieSceneScriptingFloatChannel, keys [[20,0],[23,1],[35,0.841],[41.9763,0]]}
          Scale Animation {type=MovieSceneScriptingFloatChannel, keys [[21,0.6],[23,1],[24,1.1901],[40,0.6],[45,0.5991]]}
          ColorR {type=MovieSceneScriptingFloatChannel}
          ColorG {type=MovieSceneScriptingFloatChannel}
          ColorB {type=MovieSceneScriptingFloatChannel}
          ColorA {type=MovieSceneScriptingFloatChannel}
         ]
        }]
      }]
    }
    Image_Glow {
     id=#2
     possessed_class=Image
     tracks [{
       type=MovieSceneWidgetMaterialTrack
       sections [{
         type=MovieSceneParameterSection
         end_frame=49
         start_frame=18
         channels [
          BackgroundOpacity {type=MovieSceneScriptingFloatChannel, keys [[18,0],[20,0],[26,1.7747],[40,1.5581],[49,0]]}
          ShinePosition {type=MovieSceneScriptingFloatChannel, keys [[18,0],[29,0.8784]]}
          ShineOpacity {type=MovieSceneScriptingFloatChannel, keys [[18,0],[19,0.7108],[23,0.9],[29,0.3],[31,0]]}
          ShineSize {type=MovieSceneScriptingFloatChannel, keys [[18,0.2],[23,0.35],[28,0.2],[49,0]]}
         ]
        }]
      }]
    }
   ]}
 }]
graphs [EventGraph {nodes [
   n1 = Event Set Player State!() -> (OutputDelegate, then, InPlayerState:Fort Player State Athena Object Reference) {is=title K2Node_Event}
   > SetDisplayingPlayerState!(
    execute
    self:Cherry Smoke Hero Ranking VM Object Reference=Get CherrySmokeHeroRankingVM() -> (CherrySmokeHeroRankingVM:Cherry Smoke Hero Ranking VM Object Reference)
    InPlayerState:Player State Object Reference=@n1.InPlayerState
   ) -> (then)
  ]}]
variables []
widgets {
 info {
  inheritedWidgetCount=1
  parentClass=/Script/CherrySmokeHeroRankingUI.CherrySmokeHeroRankingPlayerInfoWidget
  rootWidgetClass=&Overlay
  widgetCount=4
 }
 widgets [
  Overlay_0 {widgetClassPath=&Overlay, widget=Overlay_0}
  Image_Glow {
   widgetClassPath=&Image
   bIsVariable=true
   parent=Overlay_0
   widget=Image_Glow
   properties {
    brush {
     bIsDynamicallyLoaded=false
     drawAs=Image
     imageType=FullColor
     mirroring=NoMirror
     resourceName=None
     resourceObject=/CherrySmokeHeroRankingUI/TeammateStatus/Materials/M_UI_HeroRanking_Teammate_RankNotification_PortraitShine
     tiling=NoTile
     imageSize=<56 56>
     margin=<ltrb 0 0 0 0>
     outlineSettings {
      bUseBrushTransparency=false
      roundingType=HalfHeightRadius
      width=0
      color {colorUseRule=UseColor_Specified, specifiedColor=<rgba 0 0 0 0>}
      cornerRadii=<0 0 0 0>
     }
     tintColor {colorUseRule=UseColor_Specified, specifiedColor=<rgba 1 1 1 1>}
     uVRegion {bIsValid=false, max=<0 0>, min=<0 0>}
    }
    renderTransform {angle=0, scale=<1.1 1.1>, shear=<0 0>, translation=<0 0>}
   }
   slot {class=OverlaySlot, properties {horizontalAlignment=HAlign_Fill, verticalAlignment=VAlign_Fill}}
  }
  Image_Notification {
   widgetClassPath=&Image
   bIsVariable=true
   parent=Overlay_0
   widget=Image_Notification
   properties {
    brush {
     bIsDynamicallyLoaded=false
     drawAs=Image
     imageType=FullColor
     mirroring=NoMirror
     resourceName=None
     resourceObject=/CherrySmokeHeroRankingUI/TeammateStatus/Materials/M_UI_HeroRanking_Teammate_RankNotification
     tiling=NoTile
     imageSize=<40 40>
     margin=<ltrb 0 0 0 0>
     outlineSettings {
      bUseBrushTransparency=false
      roundingType=HalfHeightRadius
      width=0
      color {colorUseRule=UseColor_Specified, specifiedColor=<rgba 0 0 0 0>}
      cornerRadii=<0 0 0 0>
     }
     tintColor {colorUseRule=UseColor_Specified, specifiedColor=<rgba 1 1 1 1>}
     uVRegion {bIsValid=false, max=<0 0>, min=<0 0>}
    }
    renderTransform {angle=0, scale=<1.1 1.1>, shear=<0 0>, translation=<-15 6>}
   }
   slot {class=OverlaySlot, properties {verticalAlignment=VAlign_Center}}
  }
  TutorialHighlight {widgetClassPath=&TutorialHighlight}
 ]
}
